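# Unlit two-sided translucent material for the Apprentice skydiving contrail's eye glow trail, built for Niagara ribbons and mesh particles.
# World position offset widens the strip toward camera-facing with tapered widths, then adds three scrolling noise layers, two skipped at lower quality.
# Color mixes rotated bands, gradient, scrolling flow texture and inside tint, adjusted for time of day and mobile.
# Opacity combines UV edge, near-camera and depth fades with a particle-driven linear reveal.
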Material /BRCosmetics/Athena/Effects/Particles/Skydiving/NS_Contrail_Apprentice/M_Contrail_Apprentice_EyeGlowTrail
graph {
 properties {
  bCastRayTracedShadows=false
  bUseTranslucencyVertexFog=false
  bUsedWithNiagaraMeshParticles=true
  bUsedWithNiagaraRibbons=true
  blendMode=BLEND_Translucent
  pixelDepthOffsetMode=PDOM_Legacy
  shadingModel=MSM_Unlit
  twoSided=true
  usedShadingModels=Unlit
 }
 nodes [
  n1 = OneMinus(None=ComponentMask(None=TextureCoordinate()) {g=true})
  n2 = DynamicParameter() {parameterIndex=1, defaultValue=<rgba 1 1 1 0.5>, paramNames [Width,Offset,V_filp,Appear]}
  n3 = ScalarParameter() {defaultValue=3, group=02-Opacity, parameterName=Mask Size}
  n4 = StaticSwitchParameter(
   True=Frac(None=MaterialFunctionCall!(Gradient=@n1, Mask Position=@n2.Appear, Mask Size=@n3) {materialFunction=/Game/Effects/Fort_Effects/MaterialFunctions/MF_GradientMask}.Linear)
   False=Constant() {r=1}
  ) {defaultValue=true, group=02-Opacity, parameterName=Use Linear Reveal}
  n5 = Multiply(B=@n2.Width) {constA=1}
  n6 = CameraPositionWS()
  n7 = WorldPosition()
  n8 = Normalize(VectorInput=CrossProduct(A=Subtract(A=@n6, B=@n7.XYZ), B=MaterialFunctionCall() {materialFunction=&TangentBasis}.Y V))
  n9 = ScalarParameter() {defaultValue=15, group=01-Spline, parameterName=TipsWidth, sortPriority=2}
  n10 = ScalarParameter() {defaultValue=15, group=01-Spline, parameterName=MiddleWidth, sortPriority=1}
  n11 = TextureCoordinate()
  n12 = ComponentMask(None=@n11) {g=true}
  n13 = SmoothStep(Value=@n12) {constMax=0.5}
  n14 = OneMinus(None=@n12)
  n15 = Constant() {r=-0.5}
  n16 = Multiply(
   A=LinearInterpolate(
    A=ScalarParameter() {defaultValue=15, group=01-Spline, parameterName=BaseWidth, sortPriority=0}
    B=LinearInterpolate(A=@n9, B=@n10, Alpha=@n13)
    Alpha=SmoothStep(Value=@n14) {constMax=0.5}
   )
   B=ComponentMask(None=Add(A=@n15, B=@n11)) {r=true}
  )
  n17 = VertexNormalWS()
  n18 = TextureCoordinate()
  n19 = VectorParameter() {group=2-Disto, parameterName=WPO_Noise_UV, channelNames {a=SpeedY, b=SpeedX, g=ScaleY, r=ScaleX}, defaultValue=<rgba 0 0.25 0 0.25>}
  n20 = DynamicParameter() {defaultValue=<rgba 0 0 1 0>, paramNames [UVTilingRandom, TimeOffsetRandom, WPOStrength, TrailingModeON]}
  n21 = MaterialFunctionCall(X=@n19.ScaleX, Y=Add(A=@n19.ScaleY, B=@n20.UVTilingRandom)) {materialFunction=&MakeFloat2}
  n22 = MaterialFunctionCall(X=@n19.SpeedX, Y=@n19.SpeedX) {materialFunction=&MakeFloat2}
  n23 = Add(A=Time(), B=@n20.TimeOffsetRandom)
  n24 = Add(A=Multiply(A=VertexColor().R) {constB=0.5}, B=Add(A=Multiply(A=@n18, B=@n21.Result), B=Multiply(A=@n22.Result, B=@n23)))
  n25 = TextureObjectParameter() {group=5-WPO, parameterName=WPO_Tex, samplerType=SAMPLERTYPE_Grayscale, texture=&LowResBlurredNoise}
  n26 = SmoothStep(Value=TextureSample(UVs=Add(A=@n24, B=@n20.TimeOffsetRandom), Tex=@n25) {samplerSource=SSM_Wrap_WorldGroupSettings}.R) {constMax=0.7}
  n27 = TextureCoordinate()
  n28 = VectorParameter() {group=2-Disto, parameterName=WPO_TinyNoise_UV, channelNames {a=SpeedY, b=SpeedX, g=ScaleY, r=ScaleX}, defaultValue=<rgba 0 1 0 1>}
  n29 = DynamicParameter() {defaultValue=<rgba 0 0 1 0>, paramNames [UVTilingRandom, TimeOffsetRandom, WPOStrength, TrailingModeON]}
  n30 = MaterialFunctionCall(X=@n28.ScaleX, Y=Add(A=@n28.ScaleY, B=@n29.UVTilingRandom)) {materialFunction=&MakeFloat2}
  n31 = MaterialFunctionCall(X=@n28.SpeedX, Y=@n28.SpeedX) {materialFunction=&MakeFloat2}
  n32 = Add(A=@n29.TimeOffsetRandom, B=Time())
  n33 = Add(A=Multiply(A=VertexColor().R) {constB=0.5}, B=Add(A=Multiply(A=@n27, B=@n30.Result), B=Multiply(A=@n31.Result, B=@n32)))
  n34 = SmoothStep(Value=TextureSample(UVs=Add(A=@n20.TimeOffsetRandom, B=@n33), Tex=@n25) {samplerSource=SSM_Wrap_WorldGroupSettings}.R) {constMax=0.7}
  n35 = Add(
   A=Multiply(A=ConstantBiasScale(None=@n26) {bias=-0.5, scale=2}, B=ScalarParameter() {defaultValue=3, group=5-WPO, parameterName=WPO_Strength}) {constB=0.2}
   B=Multiply(A=ConstantBiasScale(None=@n34) {bias=-0.5, scale=2}, B=ScalarParameter() {defaultValue=3, group=5-WPO, parameterName=WPO_TinyStrength}) {constB=0.2}
  )
  n36 = Constant()
  n37 = VertexColor()
  n38 = VectorParameter() {group=2-Disto, parameterName=WPO_MicroNoise_UV, channelNames {a=SpeedY, b=SpeedX, g=ScaleY, r=ScaleX}, defaultValue=<rgba 0 0.5 0 0.5>}
  n39 = DynamicParameter() {defaultValue=<rgba 0 0 1 0>, paramNames [UVTilingRandom, TimeOffsetRandom, WPOStrength, TrailingModeON]}
  n40 = Multiply(
   A=TextureCoordinate()
   B=MaterialFunctionCall(X=@n38.ScaleX, Y=Add(A=@n38.ScaleY, B=@n39.UVTilingRandom)) {materialFunction=&MakeFloat2}.Result
  )
  n41 = Multiply(
   A=MaterialFunctionCall(X=@n38.SpeedX, Y=@n38.SpeedX) {materialFunction=&MakeFloat2}.Result
   B=Add(A=@n39.TimeOffsetRandom, B=Time())
  )
  n42 = Add(A=@n20.TimeOffsetRandom, B=Add(A=Multiply(A=@n37.R) {constB=0.5}, B=Add(A=@n40, B=@n41)))
  n43 = ConstantBiasScale(None=SmoothStep(Value=TextureSample(UVs=@n42, Tex=@n25) {samplerSource=SSM_Wrap_WorldGroupSettings}.R) {constMax=0.7}) {bias=-0.5, scale=2}
  n44 = ScalarParameter() {defaultValue=5, group=5-WPO, parameterName=WPO_MicroStrength}
  n45 = ComponentMask(None=TextureCoordinate()) {g=true}
  n46 = OneMinus(None=@n45)
  n47 = Multiply(
   A=Add(A=QualitySwitch(Default=@n35, Low=@n36, High=@n35, Medium=@n36, Epic=@n35), B=Multiply(A=@n43, B=@n44) {constB=0.2})
   B=Multiply(A=SmoothStep(Value=@n45) {constMax=0.75}, B=SmoothStep(Value=@n46) {constMax=0.2})
  )
  n48 = DynamicParameter() {defaultValue=<rgba 0 0 1 0>, paramNames [UVTilingRandom, TimeOffsetRandom, WPOStrength, TrailingModeON]}
  Add(A=Multiply(A=Multiply(A=@n4, B=@n5) {constA=1}, B=Multiply(A=@n8, B=@n16)), B=Multiply(A=Multiply(A=@n17, B=@n47), B=@n48.WPOStrength))
  n49 = Constant() {r=1}
  n50 = MaterialFunctionCall!(Angle=ScalarParameter() {defaultValue=0.25, group=00-Color_Setup, parameterName=RotateBand}, UV/2DVector) {materialFunction=/Game/Characters/Player/Male/Medium/Bodies/M_MED_Faux/Materials/MF_2DRotate}
  n51 = MaterialFunctionCall!(
   Width=ScalarParameter() {defaultValue=0.5, group=00-Color_Setup, parameterName=BG_BandWidth}
   Sharpness=@n49
   Offset
   Direction Switch
   Compare
   Input Coordinates=@n50.Result
  ) {materialFunction=/Engine/Functions/Engine_MaterialFunctions03/Procedurals/GeneratedBand}
  n52 = VectorParameter() {group=00-Color_Setup, parameterName=BackgroundColor, defaultValue=<rgba 0 0.3005 0.2831 0>}
  n53 = MaterialFunctionCall(Width=ScalarParameter() {defaultValue=0.3, group=00-Color_Setup, parameterName=FG_Bandwitdth}, Sharpness=@n49, Input Coordinates=@n50.Result) {materialFunction=&GeneratedBand}
  n54 = VectorParameter() {group=00-Color_Setup, parameterName=ForegroundColor, defaultValue=<rgba 0.0052 1 1 0>}
  n55 = TextureCoordinate()
  n56 = Power(
   Base=Add(A=ComponentMask(None=@n55) {g=true}, B=ScalarParameter() {group=1-ColorSetup, parameterName=ColorLerpOffset})
   Exp=ScalarParameter() {defaultValue=1, group=1-ColorSetup, parameterName=ColorExponent}
  )
  n57 = Multiply(
   A=LinearInterpolate(A=Multiply(A=@n51.Result, B=@n52.RGB) {constB=0.05}, B=Multiply(A=@n53.Result, B=@n54.RGB), Alpha=@n53.Result)
   B=LinearInterpolate(A=@n52.RGB, B=@n54.RGB, Alpha=Saturate(None=@n56))
  )
  n58 = ScalarParameter() {defaultValue=1, group=1-ColorSetup, parameterName=ColorIntensity}
  n59 = TextureCoordinate()
  n60 = VectorParameter() {group=4-Flow, parameterName=Flow UV, defaultValue=<rgba 1 1 0 3>}
  n61 = DynamicParameter() {defaultValue=<rgba 0 0 1 0>, paramNames [UVTilingRandom, TimeOffsetRandom, WPOStrength, TrailingModeON]}
  n62 = MaterialFunctionCall(X=@n60.R, Y=Add(A=@n61.UVTilingRandom, B=@n60.G)) {materialFunction=&MakeFloat2}
  n63 = MaterialFunctionCall(X=@n60.B, Y=@n60.B) {materialFunction=&MakeFloat2}
  n64 = Add(A=@n61.TimeOffsetRandom, B=Time())
  n65 = TextureSampleParameter2D(UVs=Add(A=Multiply(A=@n59, B=@n62.Result), B=Multiply(A=@n63.Result, B=@n64))) {
   group=4-Flow
   parameterName=Flow Texture
   samplerType=SAMPLERTYPE_LinearColor
   texture=/BRCosmetics/Athena/Effects/Particles/Skydiving/NS_Contrail_Turtleneck_Crystal/T_WavePattern_Chromatic
  }
  n66 = Desaturation(None=@n65.RGB, Fraction=ScalarParameter() {defaultValue=-10, group=Flow, parameterName=Flow Desaturation})
  n67 = Power(Base=@n65.A, Exp=ScalarParameter() {defaultValue=1, group=4-Flow, parameterName=FlowPower})
  n68 = ScalarParameter() {defaultValue=1, group=4-Flow, parameterName=FlowMult}
  n69 = Add(A=Multiply(A=Add(A=@n66), B=Multiply(A=@n67, B=@n68)))
  n70 = VectorParameter() {parameterName=Inside_Color, defaultValue=<rgba 0 0.9464 2 0>}
  n71 = TextureCoordinate()
  n72 = Multiply(A=Saturate(None=ComponentMask(None=@n71) {g=true}))
  n73 = SmoothStep(
   Min=ScalarParameter() {group=3-Opacity, parameterName=InsideSmoothStepMin}
   Max=ScalarParameter() {defaultValue=1, group=3-Opacity, parameterName=InsideSmoothStepMax}
   Value=@n72
  )
  n74 = ParticleColor()
  n75 = Multiply(A=Multiply(A=Multiply(A=@n57, B=@n58), B=LinearInterpolate(A=@n69, B=@n70.RGB, Alpha=@n73)), B=@n74.RGB)
  n76 = VectorParameter() {parameterName=ColorTimeOfDay, channelNames {a=Night, b=Evening, g=Day, r=Morning}, defaultValue=<rgba 1 1 1 1>}
  n77 = MaterialFunctionCall(
   Color_Morning=Multiply(A=@n75, B=@n76.Morning)
   Color_Day=Multiply(A=@n75, B=@n76.Day)
   Color_Evening=Multiply(A=@n75, B=@n76.Evening)
   Color_Night=Multiply(A=@n75, B=@n76.Night)
  ) {materialFunction=&Color_Time_of_Day}
  n78 = ScalarParameter() {defaultValue=5, parameterName=Mobile Multiplier}
  MaterialFunctionCall(Emissive=ShadingPathSwitch(Default=@n77.Result, Deferred=@n77.Result, Mobile=Multiply(A=@n78, B=@n77.Result)) {desc=Migrated from FeatureLevelSwitch}) {materialFunction=&MF_FX_EmissiveAdjustments}
  n79 = MaterialFunctionCall(Distance=Constant() {r=128}) {materialFunction=&Near_Camera_Fade}
  Multiply(A=Multiply(A=@n74.A, B=@n79.Result), B=OneMinus(None=Fresnel() {exponent=10}))
  n80 = TextureCoordinate()
  n81 = ComponentMask(None=@n80) {r=true}
  n82 = OneMinus(None=@n81)
  n83 = ComponentMask(None=@n80) {g=true}
  n84 = Multiply(A=SmoothStep(Value=@n83) {constMax=0.2}, B=SmoothStep(Value=OneMinus(None=@n83)) {constMax=0.1})
  Multiply(A=SmoothStep(Value=Multiply(A=@n81, B=@n82)) {constMax=0.05}, B=@n84)
  n85 = VectorParameter() {
   group=3-Opacity
   parameterName=Base
   channelNames {a=NoiseStrength, b=MaskMultiply, g=DepthFade, r=Refraction}
   defaultValue=<rgba 0 50 1 0.2>
  }
  n86 = Saturate(None=Multiply(A=Multiply(A=@n72, B=@n84), B=@n85.MaskMultiply))
  n87 = Multiply(A=@n79.Result, B=Saturate(None=SmoothStep(Value=@n86) {constMax=0.5}))
  n88 = DepthFade(Opacity=@n87, FadeDistance=@n85.DepthFade)
  Multiply(
   A=Multiply(
    A=ShadingPathSwitch(Default=@n88, Deferred=@n88, Forward=@n87, Mobile=@n87)
    B=ScalarParameter() {defaultValue=1, group=3-Opacity, parameterName=GlobalAlpha}
   )
   B=Saturate(None=Ceil(None=@n4))
  )
 ]
}
