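# Translucent unlit Niagara mesh-particle material that draws an animated eye for the Apprentice skydiving contrail.
# UV1 separates spline-thickened lids, tapered toward the corners, from a pupil quad expanded along instance axes and masked round.
# Dynamic parameters drive lid width, pupil scale and noise time offset, while panning noise wobbles, squeezes and pushes the lid lines.
# Gradient masks blend outer, corner, lid and pupil colors into emissive.
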Material /BRCosmetics/Athena/Effects/Particles/Skydiving/NS_Contrail_Apprentice/M_Contrail_Apprentice_Eye_Spline
graph {
 properties {
  bUsedWithNiagaraMeshParticles=true
  blendMode=BLEND_Translucent
  pixelDepthOffsetMode=PDOM_Legacy
  shadingModel=MSM_Unlit
  twoSided=true
  usedShadingModels=Unlit
 }
 nodes [
  n1 = TextureCoordinate()
  n2 = ComponentMask(None=@n1) {g=true}
  n3 = LinearInterpolate(
   A=ScalarParameter() {defaultValue=0.7, parameterName=Width Corners}
   B=ScalarParameter() {defaultValue=1, parameterName=Width Lids}
   Alpha=Sine(None=Divide(A=@n2))
  )
  n4 = DynamicParameter() {paramNames [LidScale, PupilScaleX, PupilScaleY, Time Offset]}
  n5 = VectorParameter() {parameterName=WobbleUVSettings, channelNames {a=PanY, b=PanX, g=ScaleY, r=ScaleX}, defaultValue=<rgba 0.1 0.5 0.2 0.2>}
  n6 = Multiply(
   A=ComponentMask(None=@n1) {g=true, r=true}
   B=MaterialFunctionCall(X=@n5.ScaleX, Y=@n5.ScaleX) {materialFunction=&MakeFloat2}.Result
  )
  n7 = Multiply(
   A=MaterialFunctionCall(X=@n5.PanX, Y=@n5.PanX) {materialFunction=&MakeFloat2}.Result
   B=Time()
  )
  n8 = Multiply(
   A=TextureSample(UVs=Add(A=@n6, B=@n7)) {samplerType=SAMPLERTYPE_LinearColor, texture=&HighContrastNoise}.R
   B=ScalarParameter() {parameterName=Wobbling_Strength}
  ) {constB=50}
  n9 = Multiply(A=Multiply(A=Multiply(A=@n3, B=@n4.LidScale), B=Add(A=@n8) {constA=1}), B=NamedRerouteUsage())
  n10 = MaterialFunctionCall(WidthBase=@n9, WidthTip=@n9) {materialFunction=&SplineThicken}
  n11 = MaterialFunctionCall(Float2=Subtract(A=TextureCoordinate()) {constB=0.5}) {materialFunction=&BreakOutFloat2Components}
  n12 = Multiply(A=@n11.R, B=Transform(None=Constant3Vector() {constant=<rgba 1 0 0 0>}) {transformSourceType=TRANSFORMSOURCE_Instance})
  n13 = MaterialFunctionCall(Float2=Multiply(
    A=ScalarParameter() {defaultValue=2, parameterName=Sprite Scale}
    B=MaterialFunctionCall(X=@n4.PupilScaleX, Y=@n4.PupilScaleX) {materialFunction=&MakeFloat2}.Result
   )) {materialFunction=&BreakOutFloat2Components}
  n14 = Multiply(A=@n11.G, B=Transform(None=Constant3Vector() {constant=<rgba 0 -1 0 0>}) {transformSourceType=TRANSFORMSOURCE_Instance})
  n15 = ComponentMask(None=TextureCoordinate() {coordinateIndex=1}) {r=true}
  LinearInterpolate(
   A=Add(A=@n10.WorldPositionOffset, B=NamedRerouteUsage()) {constB=0}
   B=Add(A=Multiply(A=@n12, B=@n13.R), B=Multiply(A=@n14, B=@n13.G))
   Alpha=Round(None=@n15)
  )
  n16 = Constant() {r=0.5}
  n17 = ComponentMask(None=TextureCoordinate() {coordinateIndex=1}) {r=true}
  LinearInterpolate(B=SphereMask(A=TextureCoordinate(), B=@n16, Radius=@n16, Hardness=Constant() {r=1}), Alpha=@n17) {constA=1}
  n18 = ComponentMask(None=TextureCoordinate()) {g=true}
  n19 = LinearInterpolate(
   A=VectorParameter() {parameterName=Color Corners, defaultValue=<rgba 0 0.3 1 0>}.RGB
   B=VectorParameter() {parameterName=Color Lids, defaultValue=<rgba 0.4 1.1733 2 0>}.RGB
   Alpha=Sine(None=Divide(A=@n18))
  )
  n20 = VectorParameter() {parameterName=Color Pupil, defaultValue=<rgba 0.6 2.2 3 0>}
  n21 = ComponentMask(None=@n10.UVs with Parallax) {r=true}
  n22 = LinearInterpolate(
   A=Multiply(A=OneMinus(None=@n21), B=@n21)
   B=MaterialFunctionCall() {materialFunction=&RadialGradientExponential}.RadialGradientExponential
   Alpha=@n15
  )
  MaterialFunctionCall(Emissive=LinearInterpolate(
    A=VectorParameter() {parameterName=ColorOut, defaultValue=<rgba 0 0.3005 0.2831 1>}.RGB
    B=LinearInterpolate(A=@n19, B=@n20.RGB, Alpha=@n17)
    Alpha=SmoothStep(Value=@n22) {constMax=0.3}
   )) {materialFunction=&MF_FX_EmissiveAdjustments}
  n23 = DynamicParameter() {paramNames [LidScale, PupilScaleX, PupilScaleY, Time Offset]}
  n24 = Time()
  n25 = ScalarParameter() {parameterName=Wobble Speed U}
  n26 = ScalarParameter() {parameterName=Wobble Speed V}
  n27 = TextureSample(UVs=Panner(Time=Add(A=@n23.Time Offset, B=@n24), Speed=AppendVector(A=@n25, B=@n26)) {bFractionalPart=true, constCoordinate=1}) {samplerType=SAMPLERTYPE_LinearColor, texture=&HighContrastNoise}
  n28 = Saturate(None=ScalarParameter() {parameterName=Wobble Squeeze})
  NamedRerouteDeclaration(None=OneMinus(None=Multiply(A=@n27.G, B=@n28))) {name=Wobble Squeeze, nodeColor=<rgba 0.3359 0 1 1>}
  n29 = MaterialFunctionCall(Vector=MaterialFunctionCall() {materialFunction=&LocalPosition}.Instance Local Position) {materialFunction=&SafeNormalize}
  n30 = Multiply(A=@n27.R, B=ScalarParameter() {parameterName=Wobble Intensity})
  n31 = ScalarParameter() {parameterName=Wobble Push}
  NamedRerouteDeclaration(None=Multiply(A=Transform(None=@n29.Result) {transformSourceType=TRANSFORMSOURCE_Instance}, B=Add(A=@n30, B=@n31))) {name=Wobble, nodeColor=<rgba 0 1 0.0859 1>}
 ]
}
