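# Default Lit material for the wrist mesh in the Apprentice skydiving contrail effect, enabled for Niagara mesh particles.
# Vertex color red blends the Purple and Grey color parameters, and separately blends a scalar between two constants.
# The graph also holds a UV-mapped sample of the glider wrist mesh BC texture, a U-channel mask and loose constants, with no material output connections shown.
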Material /BRCosmetics/Athena/Effects/Particles/Skydiving/NS_Contrail_Apprentice/M_Contrail_Apprentice_WristMesh
graph {
 properties {
  bUsedWithNiagaraMeshParticles=true
  pixelDepthOffsetMode=PDOM_Legacy
  previewMesh=&/SM_Contrail_Apprentice_WristMesh
  usedShadingModels=Default Lit
 }
 nodes [
  n1 = TextureCoordinate()
  ComponentMask(None=@n1) {r=true}
  Constant() {r=1}
  Constant() {r=0.7}
  n2 = VertexColor()
  LinearInterpolate(A=Constant() {r=0.5}, B=Constant() {r=0.65}, Alpha=@n2.R)
  Reroute(None=LinearInterpolate(
    A=VectorParameter() {parameterName=Purple, defaultValue=<rgba 0.5647 0.0471 0.3529 1>}.RGB
    B=VectorParameter() {parameterName=Grey, defaultValue=<rgba 0.1562 0.1562 0.1562 0>}.RGB
    Alpha=@n2.R
   ))
  TextureSample(UVs=@n1) {texture=&/T_Apprentice_Glider_WristMesh_BC}
 ]
}
