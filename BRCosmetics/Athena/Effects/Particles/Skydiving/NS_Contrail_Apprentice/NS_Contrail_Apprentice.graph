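# Niagara skydiving contrail cosmetic "Apprentice", placing glowing meshes, eye clusters and embers at two skeletal bones, apparently the wrists.
# WristMeshes samples both bones by spawn group, switching between direct placement and spring follow at a dive threshold.
# Eyes and Embers read WristMeshes through particle attribute readers; eyes sit on noisy rotated cylinders, embers spray along trail vectors.
# System.Dive drives ember rate, eye radius, scales, orientation and glow, with front-end overrides for preview.
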NiagaraSystem /BRCosmetics/Athena/Effects/Particles/Skydiving/NS_Contrail_Apprentice/NS_Contrail_Apprentice
dependencies {
 usedDataInterfaces [
  &NiagaraDataInterfaceCurve
  &NiagaraDataInterfaceColorCurve
  &NiagaraDataInterfaceParticleRead
  &NiagaraDataInterfaceVectorCurve
 ]
 usedDynamicInputs []
 usedModules [
  /Game/Effects/Niagara/Modules/Skydiving/FN_DivingAttributes/v5/FN_DivingAttributes
  &self:SystemSpawnScript.SystemScriptSource.SystemScriptGraph.NiagaraNodeAssignment_1.SetVariables_#1
  &SystemState
 ]
}
emitters {
 Embers {
  emitterUpdateScript [
   &EmitterState()
   &SpawnRate(Spawn Probability=0.5, SpawnRate=&Lerp_Float(A=0, B=20, Alpha=$System.Dive), Use Spawn Probability=true)
  ]
  particleSpawnScript [
   &FN_SetSpawnGroup(Number of Spawn Groups=2)
   &InitializeParticle(
    Color Maximum=<rgba 0 2 5 1>
    Color Minimum=<rgba 0.5782 0 5 1>
    Color Mode=::NewEnumerator2
    Lifetime Max=1.2
    Lifetime Min=0.8
    Lifetime Mode=::NewEnumerator1
    Sprite Size Mode=::NewEnumerator1
    Uniform Sprite Size Max=8
    UsePositionOffset=false
    Write Material Random=true
   ) {module=InitializeParticle001}
   &self:ReadPosAndOrientFromHandle(
    AttributeReader=DataInterface() {propertyValues {EmitterBinding {bindingMode=Other, emitterName=WristMeshes}}}
    SpawnGroup:NiagaraInt32=$Particles.SpawnGroup
   ) {module=ReadPosAndOrientFromHandle}
   /SetVariables(Particles.TrailOrientVec:Vector3f=<0 0 1>) {isSetParameters=true}
   &self:OrientVectorToVelocity(MatchingFactor=-2, RestSpeed=5, RestVector:Vector3f=<0 0 1>) {module=OrientVectorToVelocity}
   /SetVariables(Particles.Position=&TransformVector(
     Vector=&Add_Vector(
      A=&TransformVector(Vector=Linked() {linkedVariable {name=Output.ReadPosAndOrientFromHandle.HandlePos, type=NiagaraPosition}}, Destination Space=::Simulation)
      B=Linked() {linkedVariable {name=Engine.Owner.Position, type=NiagaraPosition}}
     )
     Source Space=::Simulation
     Destination Space=::World
    )) {isSetParameters=true}
   &ShapeLocation(
    MaskSpawnGroup=false
    Override Seed=false
    Rotation Mode=::NewEnumerator2
    Sphere Radius=15
    UseOwnerScale=true
    Yaw / Pitch / Roll=<0 0 0.025>
   )
   &AddVelocity(Velocity=&FN_FrontEndVector3Modifier(
     Gameplay=&Multiply_Vector(
      A=&NormalizeVector(
       Vector To Normalize=&TransformVector(Vector=$Particles.TrailOrientVec, Source Space=::Simulation, Destination Space=::Local)
       Safe Normalize=true
       Safe Fallback Threshold=0
       Safe Fallback Vector=<0 0 0.001>
      )
      B=&VectorFromFloat(Value=&RandomRangeFloat(Minimum=550, Maximum=650))
     )
     Front End=&RandomRangeVector(Minimum=<0 -200 150>, Maximum=<0 -200 250>)
    ))
  ]
  particleUpdateScript [
   &ParticleState()
   &Drag(Drag=&RandomRangeFloat(Minimum=0.5, Maximum=1), Use Linear Drag=true, Use Rotational Drag=false)
   /Niagara/Modules/Update/Forces/CurlNoiseForce(
    Noise Frequency=1
    Noise Quality / Cost=::NewEnumerator0
    Noise Strength=&FloatFromCurve(
     FloatCurve=DataInterface() {propertyValues {Curve {keys [{interpMode=RCIM_Cubic, tangentMode=RCTM_User, time=0, value=1}, {interpMode=RCIM_Cubic, tangentMode=RCTM_User, time=1, value=-1}]}}}
     Scale Curve=150
    )
    Pan Noise=false
   )
   &ScaleColor(
    Scale Alpha=&FloatFromCurve(
     FloatCurve=DataInterface() {propertyValues {Curve {keys [
         {arriveTangent=26.7273, interpMode=RCIM_Cubic, leaveTangent=26.7273, tangentMode=RCTM_User, time=0, value=0}
         {interpMode=RCIM_Cubic, tangentMode=RCTM_User, time=0.0967, value=0.9767}
         {interpMode=RCIM_Cubic, tangentMode=RCTM_User, time=0.66, value=1}
         <at 1 0 RCIM_Cubic>
        ]}}}
     Scale Curve=&RandomRangeFloat(Minimum=0.5, Maximum=2)
    )
    ScaleA=true
    ScaleRGB=true
   )
   &ScaleSpriteSize(Scale Factor=&Vector2DFromFloat(Value=&FloatFromCurve(FloatCurve=DataInterface() {propertyValues {Curve {keys [<at -0.0046 0>,<at 0.025 1>,<at 0.66 1>,<at 1.0011 0.002>]}}}, Scale Curve=0.5)))
   &InheritSourceMovement(ApplyForce=false, ApplyPosition=true, ApplyPositionFalloff=false)
   &SolveForcesAndVelocity(Clamp Velocity=false, Limit Acceleration=false)
  ]
  rendererClasses [&NiagaraSpriteRendererProperties]
  renderers [{rendererClass=&NiagaraSpriteRendererProperties}]
 }
 Eyes {
  emitterSpawnScript [&FN_PreserveSkydiveUserVariables() {module=PreserveUserVariables}]
  emitterUpdateScript [
   &EmitterState(Inactive Response=::NewEnumerator1, Life Cycle Mode=::NewEnumerator1, Loop Duration=0, UseLoopDelay=false)
   &SpawnBurst_Instantaneous(Spawn Count=16, Spawn Probability=HlslExpression() {hlslExpression="Emitter.LoopCount == 0"}, Use Spawn Probability=true) {module=SpawnBurst_Instantaneous001}
  ]
  particleSpawnScript [
   &FN_SetSpawnGroup(Number of Spawn Groups=2)
   &InitializeParticle(
    Lifetime=0
    Mass=&RandomRangeFloat(Minimum=0.4, Maximum=0.4)
    Mass Mode=::NewEnumerator1
    Mesh Scale Mode=::NewEnumerator3
    Mesh Uniform Scale=1.6
    Sprite Size Mode=::NewEnumerator3
    Uniform Sprite Size=60
    UsePositionOffset=false
    Write Material Random=true
   )
   /Niagara/Modules/Update/Mass/CalculateSizeAndRotationalInertiaByMass(Density by Material Type:ENiagaraMassByVolume=::NewEnumerator2, "Manually Entered Density (kg/m^3)"=10, Module.Manually Enter Density=false)
   &DynamicMaterialParameters(
    Index 0 Param 1=&RandomRangeFloat(Minimum=-0.2, Maximum=0.2)
    Index 0 Param 2=&RandomRangeFloat(Minimum=0, Maximum=1)
    Index 0 Param 3=&RandomRangeFloat(Minimum=15, Maximum=25)
    Index 1 Param 1=50
    Index 1 Param 2=50
    Index 1 Param 3=50
    Index 1 Param 4=1
    Index 2 Param 3=&RandomRangeFloat(Minimum=0.7, Maximum=1)
    Module.Param0WriteEnabled=true
    Module.Param0WriteEnabled1=true
    Module.Param0WriteEnabled2=true
    Module.Param1WriteEnabled=true
    Module.Param1WriteEnabled1=true
    Module.Param1WriteEnabled2=true
    Module.Param2WriteEnabled=true
    Module.Param2WriteEnabled1=true
    Module.Param2WriteEnabled2=true
    Module.Param3WriteEnabled=true
    Module.Param3WriteEnabled1=true
    Module.Param3WriteEnabled2=true
    Write Parameter Index 1=true
    Write Parameter Index 2=true
   )
   &Color()
   /SetVariables(Particles.ColorGlow:LinearColor=<rgba 0 0.7102 1 0.05>) {isSetParameters=true}
   /SetVariables(Particles.GlowSize=<80 80>) {isSetParameters=true}
   /SetVariables(Particles.TrailScale:Vector3f=&Multiply_Vector(
     A=<0.2 0.2 0.5>
     B=&MakeVector(
      X=1
      Y=1
      Z=&FloatFromCurve(
       FloatCurve=DataInterface() {propertyValues {Curve {keys [<at 0 0.2>,<at 0.45 0.7>,<at 0.55 0.7>,<at 1 0.2>]}}}
       CurveIndex=&ReturnNormalizedExecIndex()
      )
     )
    )) {isSetParameters=true}
   /SetVariables(Particles.EyeScale:Vector3f=$Particles.Scale) {isSetParameters=true}
   /SetVariables(Particles.TrailColor:LinearColor=&MakeLinearColor(
     R:NiagaraFloat=0.4
     G:NiagaraFloat=1.4
     B:NiagaraFloat=2
     A:NiagaraFloat=&FloatFromCurve(
      FloatCurve=DataInterface() {propertyValues {Curve {keys [<at 0 0.3>,<at 0.2 1>,<at 0.4 1>,<at 0.6 0.3>]}}}
      CurveIndex=&ReturnNormalizedExecIndex()
     )
    )) {isSetParameters=true}
   /SetVariables(Particles.TrailOrientVec:Vector3f=<0 0 1>) {isSetParameters=true}
   &InitialMeshOrientation(Mesh Orientation Mode=::NewEnumerator7, Rotation=<0 0 0.25>, Use Rotation Vector=true)
   &ApplyInitialForces()
  ]
  particleUpdateScript [
   &ParticleState(Kill Particles When Lifetime Has Elapsed=false)
   &ScaleMeshSize(Scale Factor=&VectorFromFloat(Value=&FloatFromCurve(FloatCurve=DataInterface() {propertyValues {Curve {keys [
          {arriveTangent=4.0031, interpMode=RCIM_Cubic, leaveTangent=4.0031, tangentMode=RCTM_User, time=0, value=0}
          {interpMode=RCIM_Cubic, tangentMode=RCTM_User, time=0.5, value=1.35}
          {interpMode=RCIM_Cubic, tangentMode=RCTM_User, time=0.75, value=1}
         ]}}})))
   &self:ReadPosAndOrientFromHandle(
    AttributeReader=DataInterface() {propertyValues {EmitterBinding {bindingMode=Other, emitterName=WristMeshes}}}
    SpawnGroup:NiagaraInt32=&SpawnGroup()
   ) {module=ReadPosAndOrientFromHandle}
   &self:RotateCylinder(
    CylinderRadius:NiagaraFloat=&Lerp_Float(
     A=&Multiply_Float(
      A=30
      B=&FloatFromCurve(
       FloatCurve=DataInterface() {propertyValues {Curve {keys [<at 0 0.8>,<at 0.45 1.1>,<at 0.55 1.1>,<at 1 0.8>]}}}
       CurveIndex=&ReturnNormalizedExecIndex()
      )
     )
     B=14
     Alpha=$System.Dive
    )
    NormalizedExecIndex:NiagaraFloat=&Add_Float(
     A=&Multiply_Float(
      A=&ReturnNormalizedExecIndex()
      B=&FloatFromCurve(FloatCurve=DataInterface() {propertyValues {Curve {keys [<at 0 0.8>,<at 1 1>]}}}, CurveIndex=$System.Dive)
     )
     B=&MaskFloatBySpawnGroup(Masked Float=0.87, Passthrough Float=0.31)
    )
    SampledRotation:Quat4f=$Output.ReadPosAndOrientFromHandle.HandleRot
   ) {module=RotateCylinder}
   &self:OffsetByQuat(
    OffsetAmount:Vector3f=&Lerp_Vector(
     A=&MaskVectorBySpawnGroup(Masked Vector=<0 0 0>, Passthrough Vector=<0 0 0>)
     B=&MaskVectorBySpawnGroup(Masked Vector=<0 0 0>, Passthrough Vector=<0 0 0>)
     Alpha=&FloatFromCurve(FloatCurve=DataInterface() {propertyValues {Curve {keys [<at 0.8 0>,<at 1 1>]}}}, CurveIndex=$System.Dive)
    )
    RotationQuat:Quat4f=$Output.ReadPosAndOrientFromHandle.HandleRot
   ) {module=OffsetByQuat}
   /Niagara/Modules/Update/Forces/V2/CurlNoiseForce(
    Noise Frequency=20
    Noise Quality / Cost=::NewEnumerator0
    Noise Strength=5
    Pan Noise=true
    Pan Noise Field=<0 0 0.5>
    Sample Position:Vector3f=&RandomVector(Vector Scale=150)
   )
   &self:OrientVectorToVelocity(
    MatchingFactor=-0.25
    RestSpeed:NiagaraFloat=&FN_FrontEndFloatModifier(Gameplay=5, Front End=25)
    RestVector:Vector3f=&FN_FrontEndVector3Modifier(Gameplay=<0 0 1>, Front End=&TransformVector(Vector=<0 -0.8 1>, Destination Space=::Simulation))
   ) {module=OrientVectorToVelocity}
   &Color(Color=&ColorFromCurve(ColorCurve=DataInterface() {propertyValues {
       AlphaCurve {keys [<at 0 1 RCIM_Cubic>,<at 1 1 RCIM_Cubic>]}
       BlueCurve {keys [<at 0 1 RCIM_Cubic>,<at 1 5 RCIM_Cubic>]}
       GreenCurve {keys [<at 0 1 RCIM_Cubic>,<at 1 5 RCIM_Cubic>]}
       RedCurve {keys [<at 0 1 RCIM_Cubic>,<at 1 5 RCIM_Cubic>]}
      }})) {module=Color001}
   &DynamicMaterialParameters(
    Index 0 Param 1=&RandomRangeFloat(Minimum=-0.1, Maximum=0.1)
    Index 0 Param 2=&RandomRangeFloat(Minimum=0, Maximum=1)
    Index 0 Param 3=&Multiply_Float(
     A=&Multiply_Float(
      A=&RandomRangeFloat(Minimum=0.9, Maximum=1.1)
      B=&FloatFromCurve(
       FloatCurve=DataInterface() {propertyValues {Curve {keys [<at 0 0.2>,<at 0.4 1>,<at 0.6 1>,<at 1 0.2>]}}}
       CurveIndex=&ReturnNormalizedExecIndex()
      )
     )
     B=&OneMinusFloat(Float=$System.Dive)
    )
    Module.Param0WriteEnabled=true
    Module.Param0WriteEnabled1=true
    Module.Param0WriteEnabled2=true
    Module.Param1WriteEnabled=true
    Module.Param1WriteEnabled1=true
    Module.Param1WriteEnabled2=true
    Module.Param2WriteEnabled=true
    Module.Param2WriteEnabled1=true
    Module.Param2WriteEnabled2=true
    Module.Param3WriteEnabled=true
    Module.Param3WriteEnabled1=true
    Module.Param3WriteEnabled2=true
    Write Parameter Index 1=true
    Write Parameter Index 2=true
   ) {module=DynamicMaterialParameters001}
   /SetVariables(Particles.TrailOrientation:Quat4f=&OrientMeshToVector(Look At Direction=$Particles.TrailOrientVec, Reference Axis Vector=<0 0 1>)) {isSetParameters=true}
   /SetVariables(Particles.TrailScale:Vector3f=&Multiply_Vector(A=$Particles.Initial.TrailScale, B=&Lerp_Vector(A=<1 1 1>, B=<1 1 2>, Alpha=$System.Dive))) {isSetParameters=true}
   /SetVariables(Particles.EyeScale:Vector3f=&Multiply_Vector(
     A=$Particles.Initial.EyeScale
     B=&Multiply_Vector(
      A=&VectorFromFloat(Value=&FloatFromCurve(
        FloatCurve=DataInterface() {propertyValues {Curve {keys [<at 0 0.6>,<at 0.45 1.2>,<at 0.55 1.2>,<at 1 0.6>]}}}
        CurveIndex=&ReturnNormalizedExecIndex()
       ))
      B=&VectorFromCurve(
       VectorCurve=DataInterface() {propertyValues {XCurve {keys [<at 0 1>,<at 0.7 0.1>]}, YCurve {keys [<at 0.75 1>,<at 1 0>]}, ZCurve {keys [<at 0 1>,<at 1 1>]}}}
       CurveIndex=$System.Dive
      )
     )
    )) {isSetParameters=true}
   /SetVariables(Particles.EyeDynamicParam:Vector4f=&MakeVector4(
     X=&MakeFloatFromVector(Vector=$Particles.Initial.EyeScale, Channel=::NewEnumerator2)
     Y=&Multiply_Float(
      A=&MakeFloatFromVector(Vector=$Particles.Initial.EyeScale, Channel=::NewEnumerator2)
      B=&FloatFromCurve(FloatCurve=DataInterface() {propertyValues {Curve {keys [<at 0 1>,<at 0.7 0>]}}}, CurveIndex=$System.Dive)
     )
     Z=&Multiply_Float(
      A=&MakeFloatFromVector(Vector=$Particles.Initial.EyeScale, Channel=::NewEnumerator2)
      B=&FloatFromCurve(FloatCurve=DataInterface() {propertyValues {Curve {keys [<at 0.7 1>,<at 0.7 0>]}}}, CurveIndex=$System.Dive)
     )
     W=&RandomRangeFloat(Minimum=0, Maximum=4)
    )) {isSetParameters=true}
   /SetVariables(Particles.Position=&Add_Vector(
     A=&Add_Vector(
      A=&TransformVector(Vector=$Output.ReadPosAndOrientFromHandle.HandlePos, Destination Space=::Simulation)
      B=Linked() {linkedVariable {name=Engine.Owner.Position, type=NiagaraPosition}}
     )
     B=&Add_Vector(
      A=&AddVectorToPosition(
       Position=&TransformVector(Vector=$Output.OffsetByQuat.OffsetVector, Destination Space=::Simulation)
       Vector=&Multiply_VectorByFloat(Vector=$Output.CurlNoiseForce.CurlNoiseForce, Float=&Lerp_Float(A=1, B=0, Alpha=$System.Dive))
      )
      B=&TransformVector(Vector=$Output.RotateCylinder.CylinderOffset, Destination Space=::Simulation)
     )
    )) {isSetParameters=true}
   /SetVariables(Particles.GlowDynamicParam:Vector4f=<0 1 1 1>) {isSetParameters=true}
   &ScaleMeshSize(Scale Factor=&VectorFromFloat(Value=&FloatFromCurve(FloatCurve=DataInterface() {propertyValues {Curve {keys [<at 0 1>,<at 1 0.7>]}}}, CurveIndex=$System.Dive))) {module=ScaleMeshSize001}
   /SetVariables(Particles.TrailDynParam1:Vector4f=&MakeVector4(
     X=&Lerp_Float(A=5, B=5, Alpha=$System.Dive)
     Y=0
     Z=0
     W=&Lerp_Float(A=0.5, B=0.5, Alpha=$System.Dive)
    )) {isSetParameters=true}
   &UpdateMeshOrientation(
    Facing Direction=&Lerp_Vector(
     A=&NormalizeVector(
      Vector To Normalize=&TransformVector(Vector=$Output.RotateCylinder.CylinderOffset, Destination Space=::Simulation)
      Safe Normalize=true
      Safe Fallback Threshold=0
      Safe Fallback Vector=<0 0 0.001>
     )
     B=&TransformVector(Vector=$Output.RotateCylinder.CylinderOffset, Destination Space=::Simulation)
     Alpha=$System.Dive
    )
    Facing Interpolation=false
    Orientation Method=::NewEnumerator1
    Override Incoming Orientation=0.4545
    Override Quat=false
    Side Direction:Vector3f=<1 0 0>
    Stabilization Mode:ENiagara_OrientVectorCount=::NewEnumerator2
    TransformFacing=false
    TransformSide=true
   ) {module=UpdateMeshOrientation001}
   &UpdateMeshOrientation(
    Rotation Rate=0.1
    Rotation Vector=&NormalizeVector(
     Vector To Normalize=&Multiply_Vector(A=$Output.CurlNoiseForce.CurlNoiseForce, B=<0 1 0>)
     Safe Normalize=true
     Safe Fallback Threshold=0
     Safe Fallback Vector=<0 0 0.001>
    )
   )
   &SolveForcesAndVelocity(Clamp Velocity=false, Limit Acceleration=false)
  ]
  rendererClasses [&NiagaraMeshRendererProperties, &NiagaraSpriteRendererProperties]
  renderers [
   {rendererClass=&NiagaraMeshRendererProperties, rendererIndex=0}
   {rendererClass=&NiagaraMeshRendererProperties, rendererIndex=1}
   {rendererClass=&NiagaraMeshRendererProperties, rendererIndex=2}
   {rendererClass=&NiagaraSpriteRendererProperties, rendererIndex=3}
  ]
 }
 WristMeshes {
  emitterSpawnScript [
   &FN_PreserveSkydiveUserVariables() {module=PreserveUserVariables}
   /Game/Effects/Niagara/MasterMaterials/ToonShader/Template/NM_ToonShaderParameters(RimLight_Color:LinearColor=<rgba 0.1615 0.1615 0.1615 1>)
   /Game/Effects/Niagara/MasterMaterials/ToonShader/Template/NM_OutlineParameters()
  ]
  emitterUpdateScript [
   &EmitterState(Inactive Response=::NewEnumerator1, Life Cycle Mode=::NewEnumerator1, Loop Duration=0, UseLoopDelay=false)
   &SpawnBurst_Instantaneous(Spawn Count=2, Spawn Probability=HlslExpression() {hlslExpression="Emitter.LoopCount == 0"}, Use Spawn Probability=true) {module=SpawnBurst_Instantaneous001}
   &SpawnBurst_Instantaneous(Spawn Count=2, Spawn Time=0.9716, Use Spawn Probability=false, UseLoopCountLimit=false)
   &SpawnBurst_Instantaneous(Spawn Count=2, Spawn Time=1.7493, Use Spawn Probability=false, UseLoopCountLimit=false) {module=SpawnBurst_Instantaneous002}
   &SpawnBurst_Instantaneous(Spawn Count=2, Spawn Time=2.5243, Use Spawn Probability=false, UseLoopCountLimit=false) {module=SpawnBurst_Instantaneous003}
  ]
  particleSpawnScript [
   &FN_SetSpawnGroup(Number of Spawn Groups=2)
   &InitializeParticle(
    Color Mode=::NewEnumerator0
    Lifetime=0
    Mesh Scale Mode=::NewEnumerator3
    Mesh Uniform Scale=1.6
    Sprite Size Mode=::NewEnumerator3
    Uniform Sprite Size=120
    UsePositionOffset=false
    Write Material Random=true
   ) {module=InitializeParticle001}
   &SampleSkeletalMeshSkeleton(
    Mesh:NiagaraDataInterfaceSkeletalMesh=$User.SkeletalMesh
    Sampled Rotation Coordinate Space:ENiagaraCoordinateSpace=::Local
    Specific Bone Index:NiagaraInt32=$Particles.SpawnGroup
   ) {module=SampleSkeletalMeshSkeleton001}
   /SetVariables(Particles.Position=$Particles.SampleSkeletalMeshSkeleton001.SampledPosition) {isSetParameters=true}
   /SetVariables(Particles.TrailOrientVec:Vector3f=<0 0 1>) {isSetParameters=true}
   &InitialMeshOrientation(
    Mesh Orientation Mode=::NewEnumerator7
    Rotation=&FN_FrontEndVector3Modifier(Gameplay=<0.25 0 0>, Front End=<0.25 0 0>)
    Use Rotation Vector=true
   )
   &Color(Color=<rgba 1 1 1 1>)
   /SetVariables(Particles.EyeScale:Vector3f=$Particles.Scale) {isSetParameters=true}
   /SetVariables(Particles.ColorGlow:LinearColor=<rgba 0 0.5 1 0.1>) {isSetParameters=true}
   &ApplyInitialForces()
  ]
  particleUpdateScript [
   &ParticleState(Kill Particles When Lifetime Has Elapsed=false)
   &SampleSkeletalMeshSkeleton(
    Mesh:NiagaraDataInterfaceSkeletalMesh=$User.SkeletalMesh
    Sampled Rotation Coordinate Space:ENiagaraCoordinateSpace=::Local
    Specific Bone Index:NiagaraInt32=$Particles.SpawnGroup
   )
   &SkeletalMeshLocation(
    Apply Sampled Orientation to Mesh Particles=true
    Apply To Particle Position=false
    Bone / Socket Index=$Particles.SpawnGroup
    Bone Sampling Mode=::NewEnumerator4
    Inherit Velocity=false
    Interpolate Mesh Sampling=false
    MaskSpawnGroup=false
    Sampled Position Offset=&MaskVectorBySpawnGroup(Masked Vector=<10 -8 0>, Passthrough Vector=<-10 8 0>)
    Sampled Position Offset Coordinate Space:ENiagaraMeshTransforms=::NewEnumerator3
    Skeletal Mesh=$User.SkeletalMesh
   )
   &self:OffsetByQuat(
    OffsetAmount:Vector3f=&Lerp_Vector(
     A=&MaskVectorBySpawnGroup(Masked Vector=<10 -8 0>, Passthrough Vector=<-10 8 0>)
     B=&MaskVectorBySpawnGroup(Masked Vector=<20 -4 0>, Passthrough Vector=<-20 4 0>)
     Alpha=$System.Dive
    )
    RotationQuat:Quat4f=$Output.SkeletalMeshLocation.SampledRotation
   ) {module=OffsetByQuat}
   /SetVariables(Particles.Position=&MakeCustomVectorFromBool(
     Bool=&SetBoolByFloatComparison(A=$System.Dive, B=0.025)
     True Vector=&Add_Vector(A=$Output.SkeletalMeshLocation.SampledPosition, B=$Output.OffsetByQuat.OffsetVector)
     False Vector=Linked() {linkedVariable {name=Particles.Position, type=NiagaraPosition}}
    )) {isSetParameters=true}
   &CurlNoiseForce(
    Noise Frequency=1
    Noise Strength=&Multiply_Float(A=150, B=&OneMinusFloat(Float=$System.Dive))
    Pan Noise=true
    Pan Noise Field=<0.5 0 0>
   )
   &self:OrientVectorToVelocity(MatchingFactor=-0.25, RestSpeed=5, RestVector:Vector3f=<0 0 1>) {module=OrientVectorToVelocity}
   /SetVariables(Particles.TrailOrientation:Quat4f=&/OrientMestToVectorFromCustomRot(
     InitialRotation:Quat4f=<0 0 0 1>
     Look At Direction:Vector3f=&TransformVector(Vector=$Particles.TrailOrientVec, Source Space=::World, Destination Space=::Simulation)
     Reference Axis Vector:Vector3f=<0 0 1>
     RotationVector:Vector3f=&FN_FrontEndVector3Modifier(Gameplay=<0 0 0>, Front End=<0.12 0 0>)
    )) {isSetParameters=true}
   &UpdateMeshOrientation(
    Rotation Rate=0
    Rotation Vector=$Output.CurlNoiseForce.CurlNoiseForce
    Write to Orientation:ENiagara_AttributeSamplingApplyOutput=::NewEnumerator1
   )
   &UpdateMeshOrientation(Rotation Rate=7, Rotation Vector=<0 1 0>) {module=UpdateMeshOrientation001}
   /SetVariables(Particles.MeshOrientation=&LerpQuaternion(
     Quaternion A:Quat4f=&LerpQuaternion(
      Quaternion A:Quat4f=$Output.UpdateMeshOrientation.MeshOrientation
      Quaternion B:Quat4f=&MultiplyQuaternion(Quaternion A=$Output.UpdateMeshOrientation.MeshOrientation, Quaternion B=<0 0.25 0 1>)
      Lerp Factor:NiagaraFloat=&FN_FrontEndFloatModifier(Gameplay=0, Front End=1)
      Enable Spherical Linear Interpolation:NiagaraBool=false
     )
     Quaternion B:Quat4f=&LerpQuaternion(
      Quaternion A:Quat4f=&MultiplyQuaternion(
       Quaternion A=&MultiplyQuaternion(Quaternion A=$Output.UpdateMeshOrientation.MeshOrientation, Quaternion B=<0 1 0 1>)
       Quaternion B=<0 0 1 1>
      )
      Quaternion B:Quat4f=&MultiplyQuaternion(
       Quaternion A=&MultiplyQuaternion(Quaternion A=$Output.UpdateMeshOrientation.MeshOrientation, Quaternion B=<0 1 0 1>)
       Quaternion B=<0 0 1 1>
      )
      Lerp Factor:NiagaraFloat=&FN_FrontEndFloatModifier(Gameplay=0, Front End=1)
      Enable Spherical Linear Interpolation:NiagaraBool=false
     )
     Lerp Factor:NiagaraFloat=&FloatFromCurve(
      FloatCurve=DataInterface() {propertyValues {Curve {keys [
          <at 0 0>
          {interpMode=RCIM_Cubic, tangentMode=RCTM_User, time=0.5, value=0}
          {arriveTangent=2.8669, interpMode=RCIM_Cubic, leaveTangent=2.8669, tangentMode=RCTM_User, time=1, value=1}
         ]}}}
      CurveIndex=$System.Dive
     )
     Enable Spherical Linear Interpolation:NiagaraBool=false
    )) {isSetParameters=true}
   /Niagara/Modules/Update/Forces/SpringForce(
    " Particle Equilibrium Position":NiagaraPosition=&Add_Vector(A=$Output.SkeletalMeshLocation.SampledPosition, B=$Output.OffsetByQuat.OffsetVector)
    Attachment Point Velocity:Vector3f=&Multiply_Vector(A=$Particles.SampleSkeletalMeshSkeleton.SampledVelocity, B=&Lerp_Vector(A=<1 1 1>, B=<1 1 1>, Alpha=$System.Dive))
    Dampening Coefficient=4
    Desired Minimum Separation Distance=0
    Force Strength:NiagaraFloat=&MakeCustomFloatFromBool(Bool=&SetBoolByFloatComparison(A=$System.Dive, B=0.025), True Float=0, False Float=3)
    Spring Tightness=22
   )
   &ScaleSpriteSize(
    Uniform Curve Index=$System.Dive
    Uniform Curve Sprite Scale=DataInterface() {propertyValues {Curve {keys [{tangentMode=RCTM_User, time=0, value=1}, {tangentMode=RCTM_User, time=1, value=2}]}}}
   )
   /SetVariables(Particles.ColorGlow:LinearColor=&ColorFromCurve(
     ColorCurve=DataInterface() {propertyValues {
       AlphaCurve {keys [<at 0 0.1 RCIM_Cubic>,<at 1 0.4 RCIM_Cubic>]}
       BlueCurve {keys [<at 0 1 RCIM_Cubic>,<at 1 1 RCIM_Cubic>]}
       GreenCurve {keys [<at 0 0.5 RCIM_Cubic>,<at 1 0.1735 RCIM_Cubic>]}
       RedCurve {keys [<at 0 0 RCIM_Cubic>,<at 1 0 RCIM_Cubic>]}
      }}
     CurveIndex=$System.Dive
    )) {isSetParameters=true}
   /SetVariables(Particles.EyeScale:Vector3f=&Multiply_Vector(A=$Particles.Initial.EyeScale, B=<1 1 1>)) {isSetParameters=true}
   /SetVariables(Particles.EyeDynamicParam:Vector4f=&MakeVector4(X=&MakeFloatFromVector(Vector=$Particles.Initial.EyeScale, Channel=::NewEnumerator2), Y=1.5, Z=1, W=0)) {isSetParameters=true}
   /SetVariables(Particles.TrailScale:Vector3f=&Lerp_Vector(
     A=<0 0 0>
     B=<1 1 2>
     Alpha=&FloatFromCurve(
      FloatCurve=DataInterface() {propertyValues {Curve {keys [
          {interpMode=RCIM_Cubic, tangentMode=RCTM_User, time=0, value=0}
          {arriveTangent=2.1274, interpMode=RCIM_Cubic, leaveTangent=2.1274, tangentMode=RCTM_User, time=1, value=1}
         ]}}}
      CurveIndex=$System.Dive
     )
    )) {isSetParameters=true}
   &DynamicMaterialParameters(
    Index 0 Param 1=&RandomRangeFloat(Minimum=0.1, Maximum=0.1)
    Index 0 Param 3=0.5
    Index 1 Param 1=1
    Index 1 Param 3=50
    Index 2 Param 2=1
    Module.Param0WriteEnabled=true
    Module.Param0WriteEnabled1=true
    Module.Param0WriteEnabled2=true
    Module.Param1WriteEnabled=true
    Module.Param1WriteEnabled1=true
    Module.Param1WriteEnabled2=true
    Module.Param2WriteEnabled=true
    Module.Param2WriteEnabled1=true
    Module.Param2WriteEnabled2=true
    Module.Param3WriteEnabled=true
    Module.Param3WriteEnabled1=true
    Module.Param3WriteEnabled2=true
    Write Parameter Index 1=true
    Write Parameter Index 2=true
   )
   /SetVariables(Particles.GlowDynamicParam:Vector4f=&MakeVector4(X=1, Y=&Lerp_Float(A=2, B=7.5, Alpha=$System.Dive), Z=1, W=1)) {isSetParameters=true}
   /SetVariables(Particles.TrailDynParam1:Vector4f=&MakeVector4(
     X=&Lerp_Float(A=0, B=15, Alpha=$System.Dive)
     Y=0
     Z=0
     W=&Lerp_Float(A=0, B=0.5, Alpha=$System.Dive)
    )) {isSetParameters=true}
   &SolveForcesAndVelocity(Clamp Velocity=false, Limit Acceleration=false)
  ]
  rendererClasses [&NiagaraMeshRendererProperties, &NiagaraSpriteRendererProperties]
  renderers [
   {rendererClass=&NiagaraMeshRendererProperties, rendererIndex=0}
   {rendererClass=&NiagaraMeshRendererProperties, rendererIndex=1}
   {rendererClass=&NiagaraMeshRendererProperties, rendererIndex=2}
   {rendererClass=&NiagaraMeshRendererProperties, rendererIndex=3}
   {rendererClass=&NiagaraSpriteRendererProperties, rendererIndex=4}
   {rendererClass=&NiagaraMeshRendererProperties, rendererIndex=5}
   {rendererClass=&NiagaraMeshRendererProperties, rendererIndex=6}
  ]
 }
}
summary {
 systemName=NS_Contrail_Apprentice
 emitters [
  {emitterName=WristMeshes, rendererClasses [&NiagaraMeshRendererProperties, &NiagaraSpriteRendererProperties]}
  {emitterName=Eyes, rendererClasses [&NiagaraMeshRendererProperties, &NiagaraSpriteRendererProperties]}
  {emitterName=Embers, rendererClasses [&NiagaraSpriteRendererProperties]}
 ]
 userVariables [
  User.bIsFrontEnd {type=NiagaraBool}
  User.bIsFrontEndPreview {type=NiagaraBool}
  User.ParaGlideLeanAlpha {type=NiagaraFloat, defaultValue=0, description=""}
  User.PlayerSpeed {type=NiagaraFloat, defaultValue=0, description=""}
  User.RibbonIDOffset {type=NiagaraInt32}
  User.SkeletalMesh {
   type=NiagaraDataInterfaceSkeletalMesh
   defaultValue {
    struct=NiagaraExt_VariableValue_DataInterface
    value {dataInterface=&self:NiagaraDataInterfaceSkeletalMesh_0}
   }
  }
 ]
}
